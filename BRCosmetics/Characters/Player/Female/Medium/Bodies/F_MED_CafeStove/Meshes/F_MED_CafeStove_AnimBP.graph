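# Post-process style AnimBlueprint for the F_MED_CafeStove cosmetic female body, which copies the pose from its attached parent mesh and corrects it with the F_MED_CafeStove_CtrlRig Control Rig.
# Each update reads crouch, emote and slide bools and converts them to float inputs for the rig.
# Crouch and emote are smoothed with FInterpTo, emote much faster, while knee sliding snaps between zero and one.
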AnimBlueprint /BRCosmetics/Characters/Player/Female/Medium/Bodies/F_MED_CafeStove/Meshes/F_MED_CafeStove_AnimBP
target_skeleton=&/F_MED_CafeStove_Skeleton
graphs [
 AnimGraph {nodes [
   n1 = Control Rig!(
    Source=Copy Pose From Mesh() {bUseAttachedParent=true}
    Alpha
    isCrouching:double=Get Is Crouching() -> (Is Crouching:float)
    IsPlayingEmote:double=Get Is Playing Emote() -> (Is Playing Emote:float)
    IsKneeSliding:double=Get Is KneeSliding() -> (Is KneeSliding:float)
   ) -> (Pose) {
    alphaBoolBlend {blendOption=Linear}
    controlRigAssetReference {blueprintRigClass=&/F_MED_CafeStove_CtrlRig.F_MED_CafeStove_CtrlRig_C}
   }
   Output Pose(Result=@n1)
  ]}
 EventGraph {nodes [
   n1 = Event BlueprintUpdateAnimation()
   > Set Delta Time X(Delta Time X=@n1.DeltaTimeX)
   n2 = Get bIsCrouching()
   > Set Is Crouching(Is Crouching:float=FInterpTo(Current=Get Is Crouching(), Target="To Float (Boolean)"(InBool=@n2), DeltaTime=Get Delta Time X(), InterpSpeed=10)) -> (Output_Get:float)
   n3 = Get bIsPlayingEmote()
   > Set Is Playing Emote(Is Playing Emote:float=FInterpTo(Current=Get Is Playing Emote(), Target="To Float (Boolean)"(InBool=@n3), DeltaTime=Get Delta Time X(), InterpSpeed=50)) -> (Output_Get:float)
   > Set Is KneeSliding(Is KneeSliding:float=SelectFloat(A=1, B=0, bPickA=Get bIsSliding())) -> (Output_Get:float)
   TryGetPawnOwner()
  ]}
]
variables [
 Delta Time X {type=number}
 Is Crouching {type=number}
 Is KneeSliding {type=number}
 Is Playing Emote {type=number}
]
