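# Physics asset for the CafeStove female medium character's FaceAcc mesh part, setting up physics for paired left and right tails and strands, likely hair.
# Three-segment tail chains hang from small master spheres, and two-segment strand chains hang from a head sphere.
# Constraints give moderate swing limits with little or zero twist.
# Unconstrained clavicle and upper-arm capsules provide body volumes, apparently for the dangling chains to collide with.
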PhysicsAsset /BRCosmetics/Characters/Player/Female/Medium/Bodies/F_MED_CafeStove/Meshes/Parts/F_MED_CafeStove_FaceAcc_Physics
bodies {
 clavicle_l [{shapeType=Capsule, length=10, radius=6.5, shapeName=clavicle_l_capsule, center=<4.1265 -1.966 0.4567>, rotation=<rot -90 180 179.9991>}]
 clavicle_r [{shapeType=Capsule, length=11, radius=7, shapeName=clavicle_r_capsule, center=<-4.1571 1.0845 -0.6033>, rotation=<rot -90 180 179.9989>}]
 dyn_strand_l_01 [{
   shapeType=Capsule
   length=6
   radius=1.5
   shapeName=dyn_strand_l_01_capsule
   center=<7.8596 -0.0257 -0.218>
   rotation=<rot 0 -96.5484 -80.2722>
  }]
 dyn_strand_l_02 [{shapeType=Capsule, length=5.5, radius=2, shapeName=dyn_strand_l_02_capsule, center=<5.194 0.814 0.1895>, rotation=<rot 0 -84.5558 -86.1575>}]
 dyn_strand_r_01 [{
   shapeType=Capsule
   length=6
   radius=1.5
   shapeName=dyn_strand_r_01_capsule
   center=<8.2366 -0.0206 -0.2091>
   rotation=<rot 0 -82.6893 -77.449>
  }]
 dyn_strand_r_02 [{
   shapeType=Capsule
   length=5.5
   radius=2
   shapeName=dyn_strand_r_02_capsule
   center=<5.0329 -0.1654 -0.0527>
   rotation=<rot 0 -87.2004 -82.1432>
  }]
 dyn_tail_01_l [{
   shapeType=Capsule
   length=8.9556
   radius=6.7677
   shapeName=dyn_tail_01_l_capsule
   center=<9.7425 0 -1.7179>
   rotation=<rot -79.9999 179.9984 -179.9983>
  }]
 dyn_tail_01_r [{shapeType=Capsule, length=9.3656, radius=6.1898, shapeName=dyn_tail_01_r_capsule, center=<8.527 0 0>, rotation=<rot 90 179.9604 179.9606>}]
 dyn_tail_02_l [{
   shapeType=Capsule
   length=7.8776
   radius=3.9449
   shapeName=dyn_tail_02_l_capsule
   center=<6.236 -0.0851 -2.5467>
   rotation=<rot -5.5647 -90 42.0796>
  }]
 dyn_tail_02_r [{
   shapeType=Capsule
   length=8.1514
   radius=3.8995
   shapeName=dyn_tail_02_r_capsule
   center=<8.6771 0.5911 -2.2072>
   rotation=<rot 1.2235 -90 26.979>
  }]
 dyn_tail_03_l [{
   shapeType=Capsule
   length=12.4177
   radius=1.3395
   shapeName=dyn_tail_03_l_capsule
   center=<3.8098 0.2519 -0.513>
   rotation=<rot 0 90.5286 -85.0059>
  }]
 dyn_tail_03_r [{
   shapeType=Capsule
   length=11.517
   radius=1.5324
   shapeName=dyn_tail_03_r_capsule
   center=<5.4201 0.1686 -1.2144>
   rotation=<rot 0 -89.8242 -82.2185>
  }]
 dyn_tail_l_mstr [{shapeType=Sphere, radius=1, shapeName=dyn_tail_l_mstr_sphere}]
 dyn_tail_r_mstr [{shapeType=Sphere, radius=1, shapeName=dyn_tail_r_mstr_sphere}]
 head [{shapeType=Sphere, radius=1, shapeName=head_sphere}]
 upperarm_l [{shapeType=Capsule, length=8, radius=7, shapeName=upperarm_l_capsule, center=<8.4471 -5.7305 -0.0001>, rotation=<rot -0.0011 -89.999 90>}]
 upperarm_r [{shapeType=Capsule, length=7.5, radius=7.5, shapeName=upperarm_r_capsule, center=<-11.6686 2.959 -3.2858>, rotation=<rot -0.0011 -89.999 90>}]
}
constraints [
 {bone1Name=dyn_tail_01_l, bone2Name=dyn_tail_l_mstr, swing1LimitDegrees=25, swing2LimitDegrees=25, twistLimitDegrees=2}
 {bone1Name=dyn_tail_01_r, bone2Name=dyn_tail_r_mstr, swing1LimitDegrees=25, swing2LimitDegrees=25, twistLimitDegrees=2}
 {bone1Name=dyn_tail_02_l, bone2Name=dyn_tail_01_l, swing1LimitDegrees=20, swing2LimitDegrees=20, twistLimitDegrees=0}
 {bone1Name=dyn_tail_03_l, bone2Name=dyn_tail_02_l, swing1LimitDegrees=20, swing2LimitDegrees=20, twistLimitDegrees=0}
 {bone1Name=dyn_tail_02_r, bone2Name=dyn_tail_01_r, swing1LimitDegrees=20, swing2LimitDegrees=20, twistLimitDegrees=0}
 {bone1Name=dyn_tail_03_r, bone2Name=dyn_tail_02_r, swing1LimitDegrees=20, swing2LimitDegrees=20, twistLimitDegrees=0}
 {bone1Name=dyn_strand_l_01, bone2Name=head, swing1LimitDegrees=20, swing2LimitDegrees=20, twistLimitDegrees=0}
 {bone1Name=dyn_strand_r_01, bone2Name=head, swing1LimitDegrees=20, swing2LimitDegrees=20, twistLimitDegrees=0}
 {bone1Name=dyn_strand_l_02, bone2Name=dyn_strand_l_01, swing1LimitDegrees=20, swing2LimitDegrees=20, twistLimitDegrees=0}
 {bone1Name=dyn_strand_r_02, bone2Name=dyn_strand_r_01, swing1LimitDegrees=20, swing2LimitDegrees=20, twistLimitDegrees=0}
]
