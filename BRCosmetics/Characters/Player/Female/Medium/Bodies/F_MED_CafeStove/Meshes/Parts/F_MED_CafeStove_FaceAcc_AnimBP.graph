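# Post-process-style Anim Blueprint for the CafeStove female character's face accessory part, adding secondary motion to tails, strands, bangs and bows.
# It copies the parent mesh pose, layering in BodyMesh from the clavicles.
# bRigidBodyEnabled selects RigidBody simulation, with a backpack-specific physics asset, or pre-rotated tail bones plus AnimDynamics; bang, back-strand and bow AnimDynamics always follow.
# The event graph updates backpack, gravity and head-tilt variables.
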AnimBlueprint /BRCosmetics/Characters/Player/Female/Medium/Bodies/F_MED_CafeStove/Meshes/Parts/F_MED_CafeStove_FaceAcc_AnimBP
target_skeleton=&/F_MED_CafeStove_FaceAcc_Skeleton
graphs [
 AnimGraph {nodes [
   Anim Dynamics - Bone: dyn_tail_02_r!(ComponentPose, Alpha=1, GravityOverride:Vector="0,0,0", ExternalForce="0,0,0") -> (Pose) {
    angularDampingOverride=0.9999
    bOverrideAngularDamping=true
    bUseGravityOverride=true
    bUsePlanarLimit=false
    bUseSphericalLimits=true
    gravityScale=0.35
    alphaBoolBlend {blendOption=Linear}
    boundBone {boneName=dyn_tail_02_r}
    componentAppliedLinearAccClamp=<700 700 700>
    componentLinearAccScale=<0.3 0.3 0.3>
    componentLinearVelScale=<0.4 0.4 0.4>
    physicsBodyDefinitions [{
      sphereCollisionRadius=1
      boundBone {boneName=dyn_tail_02_r}
      boxExtents=<5 15 20>
      constraintSetup {angularLimitsMax=<0 10 15>, angularLimitsMin=<0 -15 -15>, angularTarget=<0 0 0>}
      localJointOffset=<2.5 0 0>
     }]
    sphericalLimits [
     {limitRadius=12, drivingBone {boneName=clavicle_r}, sphereLocalOffset=<-12.4054 -2.3305 0>}
     {limitRadius=10, drivingBone {boneName=upperarm_r}, sphereLocalOffset=<-10.9572 0 -1.6264>}
    ]
   }
   Anim Dynamics - Bone: dyn_tail_01_r!(ComponentPose, Alpha=1, GravityOverride:Vector="0,0,0", ExternalForce="0,0,0") -> (Pose) {
    angularDampingOverride=0.9999
    bOverrideAngularDamping=true
    bUseGravityOverride=true
    bUsePlanarLimit=false
    bUseSphericalLimits=true
    alphaBoolBlend {blendOption=Linear}
    boundBone {boneName=dyn_tail_01_r}
    componentAppliedLinearAccClamp=<700 700 700>
    componentLinearAccScale=<0.3 0.3 0.3>
    componentLinearVelScale=<0.4 0.4 0.4>
    physicsBodyDefinitions [{
      sphereCollisionRadius=1
      boundBone {boneName=dyn_tail_01_r}
      boxExtents=<12 10 10>
      constraintSetup {angularLimitsMax=<0 10 10>, angularLimitsMin=<0 -10 -10>, angularTarget=<0 0 0>}
      localJointOffset=<1 0 0>
     }]
    sphericalLimits [
     {limitRadius=10, drivingBone {boneName=spine_05}, sphereLocalOffset=<4 -5 10>}
     {limitRadius=10, drivingBone {boneName=upperarm_r}, sphereLocalOffset=<-9.1967 0 -9.3087>}
    ]
   }
   n1 = Use cached pose 'RBAN'()
   n2 = Use cached pose 'AnimDynamics'()
   n3 = Get bRigidBodyEnabled()
   n4 = Anim Dynamics - Bone: dyn_bang_01!(
    ComponentPose=Local To Component(LocalPose=Blend Poses by bool(BlendPose_0=@n1, BlendPose_1=@n2, bActiveValue=@n3) {blendType=Linear})
    Alpha=1
    GravityOverride:Vector="0,0,0"
    ExternalForce="0,0,0"
   ) -> (Pose) {
    angularDampingOverride=0.9999
    bOverrideAngularDamping=true
    bUseGravityOverride=true
    bUsePlanarLimit=false
    alphaBoolBlend {blendOption=Linear}
    boundBone {boneName=dyn_bang_01}
    componentAppliedLinearAccClamp=<700 700 700>
    componentLinearAccScale=<0.3 0.3 0.3>
    componentLinearVelScale=<0.4 0.4 0.4>
    physicsBodyDefinitions [{
      sphereCollisionRadius=1
      boundBone {boneName=dyn_bang_01}
      boxExtents=<15 10 10>
      constraintSetup {angularLimitsMax=<0 0 15>, angularLimitsMin=<0 -15 -15>, angularTarget=<0 0 0>}
      localJointOffset=<2 0 0>
     }]
   }
   n5 = Anim Dynamics - Bone: dyn_bang_02!(ComponentPose=@n4, Alpha=1, GravityOverride:Vector="0,0,0", ExternalForce="0,0,0") -> (Pose) {
    angularDampingOverride=0.9999
    bOverrideAngularDamping=true
    bUseGravityOverride=true
    bUsePlanarLimit=false
    gravityScale=0.35
    alphaBoolBlend {blendOption=Linear}
    boundBone {boneName=dyn_bang_02}
    componentAppliedLinearAccClamp=<700 700 700>
    componentLinearAccScale=<0.3 0.3 0.3>
    componentLinearVelScale=<0.4 0.4 0.4>
    physicsBodyDefinitions [{
      sphereCollisionRadius=1
      boundBone {boneName=dyn_bang_02}
      boxExtents=<5 15 20>
      constraintSetup {angularLimitsMax=<0 0 15>, angularLimitsMin=<0 -15 -15>, angularTarget=<0 0 0>}
      localJointOffset=<2.5 0 0>
     }]
   }
   n6 = Anim Dynamics - Bone: dyn_bang_r_01!(ComponentPose=@n5, Alpha=1, GravityOverride:Vector="0,0,0", ExternalForce="0,0,0") -> (Pose) {
    angularDampingOverride=0.9999
    bOverrideAngularDamping=true
    bUseGravityOverride=true
    bUsePlanarLimit=false
    gravityScale=0.35
    alphaBoolBlend {blendOption=Linear}
    boundBone {boneName=dyn_bang_r_01}
    componentAppliedLinearAccClamp=<700 700 700>
    componentLinearAccScale=<0.3 0.3 0.3>
    componentLinearVelScale=<0.4 0.4 0.4>
    physicsBodyDefinitions [{
      sphereCollisionRadius=1
      boundBone {boneName=dyn_bang_r_01}
      boxExtents=<5 10 10>
      constraintSetup {angularLimitsMax=<0 0 30>, angularLimitsMin=<0 -15 -5>, angularTarget=<0 0 0>}
      localJointOffset=<2 0 0>
     }]
   }
   n7 = Anim Dynamics - Bone: dyn_bang_l_01!(ComponentPose=@n6, Alpha=1, GravityOverride:Vector="0,0,0", ExternalForce="0,0,0") -> (Pose) {
    angularDampingOverride=0.9999
    bOverrideAngularDamping=true
    bUseGravityOverride=true
    bUsePlanarLimit=false
    gravityScale=0.35
    alphaBoolBlend {blendOption=Linear}
    boundBone {boneName=dyn_bang_l_01}
    componentAppliedLinearAccClamp=<700 700 700>
    componentLinearAccScale=<0.3 0.3 0.3>
    componentLinearVelScale=<0.4 0.4 0.4>
    physicsBodyDefinitions [{
      sphereCollisionRadius=1
      boundBone {boneName=dyn_bang_l_01}
      boxExtents=<5 10 10>
      constraintSetup {angularLimitsMax=<0 0 5>, angularLimitsMin=<0 -15 -30>, angularTarget=<0 0 0>}
      localJointOffset=<2 0 0>
     }]
   }
   n8 = Anim Dynamics - Bone: dyn_strand_back_l!(ComponentPose=@n7, Alpha=1, GravityOverride:Vector="0,0,0", ExternalForce="0,0,0") -> (Pose) {
    angularDampingOverride=0.99
    bOverrideAngularDamping=true
    bUseGravityOverride=true
    bUsePlanarLimit=false
    gravityScale=0.35
    alphaBoolBlend {blendOption=Linear}
    boundBone {boneName=dyn_strand_back_l}
    componentAppliedLinearAccClamp=<700 700 700>
    componentLinearAccScale=<0.3 0.3 0.3>
    componentLinearVelScale=<0.4 0.4 0.4>
    physicsBodyDefinitions [{
      sphereCollisionRadius=1
      boundBone {boneName=dyn_strand_back_l}
      boxExtents=<7 15 20>
      constraintSetup {angularLimitsMax=<0 0 30>, angularLimitsMin=<0 -30 -30>, angularTarget=<0 0 0>}
      localJointOffset=<4 0 0>
     }]
   }
   n9 = Anim Dynamics - Bone: dyn_strand_back_r!(ComponentPose=@n8, Alpha=1, GravityOverride:Vector="0,0,0", ExternalForce="0,0,0") -> (Pose) {
    angularDampingOverride=0.99
    bOverrideAngularDamping=true
    bUseGravityOverride=true
    bUsePlanarLimit=false
    gravityScale=0.35
    alphaBoolBlend {blendOption=Linear}
    boundBone {boneName=dyn_strand_back_r}
    componentAppliedLinearAccClamp=<700 700 700>
    componentLinearAccScale=<0.3 0.3 0.3>
    componentLinearVelScale=<0.4 0.4 0.4>
    physicsBodyDefinitions [{
      sphereCollisionRadius=1
      boundBone {boneName=dyn_strand_back_r}
      boxExtents=<7 10 10>
      constraintSetup {angularLimitsMax=<0 30 30>, angularLimitsMin=<0 0 -30>, angularTarget=<0 0 0>}
      localJointOffset=<3 0 0>
     }]
   }
   n10 = Anim Dynamics - Bone: dyn_bow_down_r(ComponentPose=@n9, Alpha=1, ExternalForce="0,0,0") {
    angularDampingOverride=0.9999
    bOverrideAngularDamping=true
    bUseGravityOverride=true
    bUsePlanarLimit=false
    gravityScale=0.35
    alphaBoolBlend {blendOption=Linear}
    boundBone {boneName=dyn_bow_down_r}
    componentAppliedLinearAccClamp=<500 500 500>
    componentLinearAccScale=<0.2 0.2 0.2>
    componentLinearVelScale=<0.2 0.2 0.2>
    gravityOverride {z=-200}
    physicsBodyDefinitions [{
      sphereCollisionRadius=1
      boundBone {boneName=dyn_bow_down_r}
      boxExtents=<10 15 15>
      constraintSetup {angularLimitsMax=<0 5 5>, angularLimitsMin=<0 -5 -25>, angularTarget=<0 0 0>}
      localJointOffset=<2.5 0 0>
     }]
   }
   n11 = Anim Dynamics - Bone: dyn_bow_up_r!(ComponentPose=@n10, Alpha=1, GravityOverride:Vector="0,0,-500", ExternalForce="0,0,0") -> (Pose) {
    angularDampingOverride=0.9999
    bOverrideAngularDamping=true
    bUseGravityOverride=true
    bUsePlanarLimit=false
    simulationSpace=RootRelative
    alphaBoolBlend {blendOption=Linear}
    boundBone {boneName=dyn_bow_up_r}
    componentAppliedLinearAccClamp=<200 200 200>
    componentLinearAccScale=<0.2 0.2 0.2>
    componentLinearVelScale=<0.2 0.2 0.2>
    gravityOverride {z=-500}
    physicsBodyDefinitions [{
      sphereCollisionRadius=1
      boundBone {boneName=dyn_bow_up_r}
      boxExtents=<15 10 10>
      constraintSetup {angularLimitsMax=<0 5 20>, angularLimitsMin=<0 -5 -15>, angularTarget=<0 0 0>}
      localJointOffset=<0.5 0 0>
     }]
   }
   n12 = Anim Dynamics - Bone: dyn_bow_down_l(ComponentPose=@n11, Alpha=1, ExternalForce="0,0,0") {
    angularDampingOverride=0.9999
    bOverrideAngularDamping=true
    bUseGravityOverride=true
    bUsePlanarLimit=false
    gravityScale=0.35
    alphaBoolBlend {blendOption=Linear}
    boundBone {boneName=dyn_bow_down_l}
    componentAppliedLinearAccClamp=<500 500 500>
    componentLinearAccScale=<0.2 0.2 0.2>
    componentLinearVelScale=<0.2 0.2 0.2>
    gravityOverride {z=-200}
    physicsBodyDefinitions [{
      sphereCollisionRadius=1
      boundBone {boneName=dyn_bow_down_l}
      boxExtents=<10 15 15>
      constraintSetup {angularLimitsMax=<0 5 25>, angularLimitsMin=<0 -5 -5>, angularTarget=<0 0 0>}
      localJointOffset=<2.5 0 0>
     }]
   }
   n13 = Anim Dynamics - Bone: dyn_bow_up_l!(ComponentPose=@n12, Alpha=1, GravityOverride:Vector="0,0,-500", ExternalForce="0,0,0") -> (Pose) {
    angularDampingOverride=0.9999
    bOverrideAngularDamping=true
    bUseGravityOverride=true
    bUsePlanarLimit=false
    alphaBoolBlend {blendOption=Linear}
    boundBone {boneName=dyn_bow_up_l}
    componentAppliedLinearAccClamp=<200 200 200>
    componentLinearAccScale=<0.2 0.2 0.2>
    componentLinearVelScale=<0.2 0.2 0.2>
    gravityOverride {z=-500}
    physicsBodyDefinitions [{
      sphereCollisionRadius=1
      boundBone {boneName=dyn_bow_up_l}
      boxExtents=<15 10 10>
      constraintSetup {angularLimitsMax=<0 5 15>, angularLimitsMin=<0 -5 -20>, angularTarget=<0 0 0>}
      localJointOffset=<0.5 0 0>
     }]
   }
   Output Pose(Result=Component To Local(ComponentPose=@n13))
   n14 = Get BodyMesh() -> (BodyMesh:Skeletal Mesh Component Budgeted Object Reference)
   Save cached pose 'InitialPose'(Pose=Layered blend per bone(
     BasePose=Copy Pose From Mesh() {bUseAttachedParent=true}
     BlendPoses_0=Copy Pose From Mesh(SourceMeshComponent=@n14) {bUseAttachedParent=true}
    ) {layerSetup [{branchFilters [{boneName=clavicle_l}, {boneName=clavicle_r}]}]}) {cacheName=InitialPose}
   n15 = RigidBody!(ComponentPose=Local To Component(LocalPose=Use cached pose 'InitialPose'()), Alpha, OverrideWorldGravity:Vector="0,0,0", ExternalForce) -> (Pose) {
    bOverrideWorldGravity=true
    overridePhysicsAsset=&/F_MED_CafeStove_FaceAcc_Backpack_Physics
    simulationSpace=BaseBoneSpace
    alphaBoolBlend {blendOption=Linear}
    baseBoneRef {boneName=spine_05}
    componentAppliedLinearAccClamp=<1500 1500 1500>
    componentLinearAccScale=<0.3 0.3 0.3>
    componentLinearVelScale=<0.3 0.3 0.3>
   }
   n16 = RigidBody(ComponentPose=Local To Component(LocalPose=Use cached pose 'InitialPose'()), OverrideWorldGravity="0,0,0") {
    bOverrideWorldGravity=true
    overridePhysicsAsset=&/F_MED_CafeStove_FaceAcc_Physics
    simulationSpace=BaseBoneSpace
    alphaBoolBlend {blendOption=Linear}
    baseBoneRef {boneName=spine_05}
    componentAppliedLinearAccClamp=<1500 1500 1500>
    componentLinearAccScale=<0.3 0.3 0.3>
    componentLinearVelScale=<0.3 0.3 0.3>
   }
   Save cached pose 'RBAN'(Pose=Blend Poses by bool(
     BlendPose_0=Component To Local(ComponentPose=@n15)
     BlendPose_1=Component To Local(ComponentPose=@n16)
     bActiveValue=Get bIsWearingBackpack() -> (bIsWearingBackpack:bool)
    ) {blendType=Linear}) {cacheName=RBAN}
   n17 = "Transform (Modify) Bone - Bone: dyn_tail_r_mstr"(ComponentPose=Local To Component(LocalPose=Use cached pose 'InitialPose'()), Alpha=1, Translation="0,0,0", Rotation="45,0,0", Scale="1,1,1") {
    rotationMode=BMM_Additive
    rotationSpace=BCS_BoneSpace
    translationSpace=BCS_ParentBoneSpace
    alphaBoolBlend {blendOption=Linear}
    boneToModify {boneName=dyn_tail_r_mstr}
    rotation {pitch=45}
   }
   n18 = "Transform (Modify) Bone - Bone: dyn_tail_l_mstr"(ComponentPose=@n17, Alpha=1, Translation="0,0,0", Rotation="15,0,0", Scale="1,1,1") {
    rotationMode=BMM_Additive
    rotationSpace=BCS_BoneSpace
    translationSpace=BCS_ParentBoneSpace
    alphaBoolBlend {blendOption=Linear}
    boneToModify {boneName=dyn_tail_l_mstr}
    rotation {pitch=15}
   }
   n19 = "Transform (Modify) Bone - Bone: dyn_tail_l_mstr"(ComponentPose=@n18, Alpha=1, Translation="0,0,0", Rotation="0,-15,0", Scale="1,1,1") {
    rotationMode=BMM_Additive
    rotationSpace=BCS_BoneSpace
    translationSpace=BCS_ParentBoneSpace
    alphaBoolBlend {blendOption=Linear}
    boneToModify {boneName=dyn_tail_l_mstr}
    rotation {yaw=-15}
   }
   n20 = "Transform (Modify) Bone - Bone: dyn_tail_l_mstr"(ComponentPose=@n19, Alpha=1, Translation="0,0,0", Rotation="0,-30,0", Scale="1,1,1") {
    rotationMode=BMM_Additive
    rotationSpace=BCS_BoneSpace
    translationSpace=BCS_ParentBoneSpace
    alphaBoolBlend {blendOption=Linear}
    boneToModify {boneName=dyn_tail_l_mstr}
    rotation {yaw=-30}
   }
   n21 = "Transform (Modify) Bone - Bone: dyn_tail_l_mstr"(ComponentPose=@n20, Alpha=1, Translation="0,0,0", Rotation="15,-45,0", Scale="1,1,1") {
    rotationMode=BMM_Additive
    rotationSpace=BCS_BoneSpace
    translationSpace=BCS_ParentBoneSpace
    alphaBoolBlend {blendOption=Linear}
    boneToModify {boneName=dyn_tail_l_mstr}
    rotation {pitch=15, yaw=-45}
   }
   n22 = "Transform (Modify) Bone - Bone: dyn_tail_r_mstr"(ComponentPose=@n21, Alpha=1, Translation="0,0,0", Rotation="0,-45,0", Scale="1,1,1") {
    rotationMode=BMM_Additive
    rotationSpace=BCS_BoneSpace
    translationSpace=BCS_ParentBoneSpace
    alphaBoolBlend {blendOption=Linear}
    boneToModify {boneName=dyn_tail_r_mstr}
    rotation {yaw=-45}
   }
   n23 = Anim Dynamics - Bone: dyn_strand_r_01!(ComponentPose=@n22, Alpha=1, GravityOverride:Vector="0,0,0", ExternalForce="0,0,0") -> (Pose) {
    angularDampingOverride=0.9999
    bOverrideAngularDamping=true
    bUseGravityOverride=true
    bUsePlanarLimit=false
    alphaBoolBlend {blendOption=Linear}
    boundBone {boneName=dyn_strand_r_01}
    componentAppliedLinearAccClamp=<700 700 700>
    componentLinearAccScale=<0.3 0.3 0.3>
    componentLinearVelScale=<0.4 0.4 0.4>
    physicsBodyDefinitions [{
      sphereCollisionRadius=1
      boundBone {boneName=dyn_strand_r_01}
      boxExtents=<15 10 10>
      constraintSetup {angularLimitsMax=<0 5 15>, angularLimitsMin=<0 -10 -15>, angularTarget=<0 0 0>}
      localJointOffset=<2 0 0>
     }]
   }
   n24 = Anim Dynamics - Bone: dyn_strand_r_02!(ComponentPose=@n23, Alpha=1, GravityOverride:Vector="0,0,0", ExternalForce="0,0,0") -> (Pose) {
    angularDampingOverride=0.9999
    bOverrideAngularDamping=true
    bUseGravityOverride=true
    bUsePlanarLimit=false
    bUseSphericalLimits=true
    gravityScale=0.35
    alphaBoolBlend {blendOption=Linear}
    boundBone {boneName=dyn_strand_r_02}
    componentAppliedLinearAccClamp=<700 700 700>
    componentLinearAccScale=<0.3 0.3 0.3>
    componentLinearVelScale=<0.4 0.4 0.4>
    physicsBodyDefinitions [{
      sphereCollisionRadius=1
      boundBone {boneName=dyn_strand_r_02}
      boxExtents=<15 5 5>
      constraintSetup {angularLimitsMax=<0 5 15>, angularLimitsMin=<0 -10 -15>, angularTarget=<0 0 0>}
      localJointOffset=<2.5 0 0>
     }]
    sphericalLimits [
     {limitRadius=10, drivingBone {boneName=clavicle_r}, sphereLocalOffset=<-4.3518 3.9801 -0.9712>}
     {limitRadius=7, drivingBone {boneName=upperarm_r}, sphereLocalOffset=<-11.3089 0 0>}
    ]
   }
   n25 = Anim Dynamics - Bone: dyn_strand_l_01!(ComponentPose=@n24, Alpha=1, GravityOverride:Vector="0,0,0", ExternalForce="0,0,0") -> (Pose) {
    angularDampingOverride=0.9999
    bOverrideAngularDamping=true
    bUseGravityOverride=true
    bUsePlanarLimit=false
    alphaBoolBlend {blendOption=Linear}
    boundBone {boneName=dyn_strand_l_01}
    componentAppliedLinearAccClamp=<700 700 700>
    componentLinearAccScale=<0.3 0.3 0.3>
    componentLinearVelScale=<0.4 0.4 0.4>
    physicsBodyDefinitions [{
      sphereCollisionRadius=1
      boundBone {boneName=dyn_strand_l_01}
      boxExtents=<15 10 10>
      constraintSetup {angularLimitsMax=<0 5 15>, angularLimitsMin=<0 -10 -15>, angularTarget=<0 0 0>}
      localJointOffset=<2 0 0>
     }]
   }
   n26 = Anim Dynamics - Bone: dyn_strand_l_02!(ComponentPose=@n25, Alpha=1, GravityOverride:Vector="0,0,0", ExternalForce="0,0,0") -> (Pose) {
    angularDampingOverride=0.9999
    bOverrideAngularDamping=true
    bUseGravityOverride=true
    bUsePlanarLimit=false
    bUseSphericalLimits=true
    gravityScale=0.35
    alphaBoolBlend {blendOption=Linear}
    boundBone {boneName=dyn_strand_l_02}
    componentAppliedLinearAccClamp=<700 700 700>
    componentLinearAccScale=<0.3 0.3 0.3>
    componentLinearVelScale=<0.4 0.4 0.4>
    physicsBodyDefinitions [{
      sphereCollisionRadius=1
      boundBone {boneName=dyn_strand_l_02}
      boxExtents=<15 5 5>
      constraintSetup {angularLimitsMax=<0 5 15>, angularLimitsMin=<0 -10 -15>, angularTarget=<0 0 0>}
      localJointOffset=<2.5 0 0>
     }]
    sphericalLimits [{limitRadius=10, drivingBone {boneName=clavicle_l}, sphereLocalOffset=<3.136 -2.5896 2.1795>}]
   }
   n27 = Anim Dynamics - Bone: dyn_tail_01_r(ComponentPose=@n26, Alpha=1, GravityOverride="0,0,0", ExternalForce="0,0,0") {
    angularDampingOverride=0.9999
    bOverrideAngularDamping=true
    bUseGravityOverride=true
    bUsePlanarLimit=false
    alphaBoolBlend {blendOption=Linear}
    boundBone {boneName=dyn_tail_01_r}
    componentAppliedLinearAccClamp=<700 700 700>
    componentLinearAccScale=<0.3 0.3 0.3>
    componentLinearVelScale=<0.4 0.4 0.4>
    physicsBodyDefinitions [{
      sphereCollisionRadius=1
      boundBone {boneName=dyn_tail_01_r}
      boxExtents=<12 10 10>
      constraintSetup {angularLimitsMax=<0 0 10>, angularLimitsMin=<0 -30 -10>, angularTarget=<0 0 0>}
      localJointOffset=<1 0 0>
     }]
   }
   n28 = Anim Dynamics - Bone: dyn_tail_02_r(ComponentPose=@n27, Alpha=1, GravityOverride="0,0,0", ExternalForce="0,0,0") {
    angularDampingOverride=0.9999
    bOverrideAngularDamping=true
    bUseGravityOverride=true
    bUsePlanarLimit=false
    bUseSphericalLimits=true
    gravityScale=0.35
    alphaBoolBlend {blendOption=Linear}
    boundBone {boneName=dyn_tail_02_r}
    componentAppliedLinearAccClamp=<700 700 700>
    componentLinearAccScale=<0.3 0.3 0.3>
    componentLinearVelScale=<0.4 0.4 0.4>
    physicsBodyDefinitions [{
      sphereCollisionRadius=1
      boundBone {boneName=dyn_tail_02_r}
      boxExtents=<5 15 20>
      constraintSetup {angularLimitsMax=<0 10 15>, angularLimitsMin=<0 -15 -15>, angularTarget=<0 0 0>}
      localJointOffset=<2.5 0 0>
     }]
    sphericalLimits [{limitRadius=10, drivingBone {boneName=upperarm_r}, sphereLocalOffset=<-10.9572 0 -1.6264>}]
   }
   n29 = Anim Dynamics - Bone: dyn_tail_03_r!(ComponentPose=@n28, Alpha=1, GravityOverride:Vector="0,0,0", ExternalForce="0,0,0") -> (Pose) {
    angularDampingOverride=0.9999
    bOverrideAngularDamping=true
    bUseGravityOverride=true
    bUsePlanarLimit=false
    bUseSphericalLimits=true
    gravityScale=0.35
    alphaBoolBlend {blendOption=Linear}
    boundBone {boneName=dyn_tail_03_r}
    componentAppliedLinearAccClamp=<700 700 700>
    componentLinearAccScale=<0.3 0.3 0.3>
    componentLinearVelScale=<0.4 0.4 0.4>
    physicsBodyDefinitions [{
      sphereCollisionRadius=1
      boundBone {boneName=dyn_tail_03_r}
      boxExtents=<5 10 10>
      constraintSetup {angularLimitsMax=<0 5 30>, angularLimitsMin=<0 -15 -5>, angularTarget=<0 0 0>}
      localJointOffset=<2 0 0>
     }]
    sphericalLimits [{limitRadius=10, drivingBone {boneName=clavicle_r}, sphereLocalOffset=<-7.3876 -2.314 0>}]
   }
   n30 = Anim Dynamics - Bone: dyn_tail_01_l!(ComponentPose=@n29, Alpha=1, GravityOverride:Vector="0,0,0", ExternalForce="0,0,0") -> (Pose) {
    angularDampingOverride=0.9999
    bOverrideAngularDamping=true
    bUseGravityOverride=true
    bUsePlanarLimit=false
    bUseSphericalLimits=true
    alphaBoolBlend {blendOption=Linear}
    boundBone {boneName=dyn_tail_01_l}
    componentAppliedLinearAccClamp=<700 700 700>
    componentLinearAccScale=<0.3 0.3 0.3>
    componentLinearVelScale=<0.4 0.4 0.4>
    physicsBodyDefinitions [{
      sphereCollisionRadius=1
      boundBone {boneName=dyn_tail_01_l}
      boxExtents=<12 10 10>
      constraintSetup {angularLimitsMax=<0 30 10>, angularLimitsMin=<0 0 -10>, angularTarget=<0 0 0>}
      localJointOffset=<1 0 0>
     }]
    sphericalLimits [{limitRadius=10, drivingBone {boneName=spine_05}, sphereLocalOffset=<0 -5 -10>}]
   }
   n31 = Anim Dynamics - Bone: dyn_tail_02_l!(ComponentPose=@n30, Alpha=1, GravityOverride:Vector="0,0,0", ExternalForce="0,0,0") -> (Pose) {
    angularDampingOverride=0.9999
    bOverrideAngularDamping=true
    bUseGravityOverride=true
    bUsePlanarLimit=false
    bUseSphericalLimits=true
    gravityScale=0.35
    alphaBoolBlend {blendOption=Linear}
    boundBone {boneName=dyn_tail_02_l}
    componentAppliedLinearAccClamp=<700 700 700>
    componentLinearAccScale=<0.3 0.3 0.3>
    componentLinearVelScale=<0.4 0.4 0.4>
    physicsBodyDefinitions [{
      sphereCollisionRadius=1
      boundBone {boneName=dyn_tail_02_l}
      boxExtents=<5 15 20>
      constraintSetup {angularLimitsMax=<0 10 15>, angularLimitsMin=<0 -15 -15>, angularTarget=<0 0 0>}
      localJointOffset=<2.5 0 0>
     }]
    sphericalLimits [{limitRadius=10, drivingBone {boneName=clavicle_l}, sphereLocalOffset=<9.9818 0 0>}]
   }
   n32 = Anim Dynamics - Bone: dyn_tail_03_l!(ComponentPose=@n31, Alpha=1, GravityOverride:Vector="0,0,0", ExternalForce="0,0,0") -> (Pose) {
    angularDampingOverride=0.9999
    bOverrideAngularDamping=true
    bUseGravityOverride=true
    bUsePlanarLimit=false
    gravityScale=0.35
    alphaBoolBlend {blendOption=Linear}
    boundBone {boneName=dyn_tail_03_l}
    componentAppliedLinearAccClamp=<700 700 700>
    componentLinearAccScale=<0.3 0.3 0.3>
    componentLinearVelScale=<0.4 0.4 0.4>
    physicsBodyDefinitions [{
      sphereCollisionRadius=1
      boundBone {boneName=dyn_tail_03_l}
      boxExtents=<5 10 10>
      constraintSetup {angularLimitsMax=<0 5 30>, angularLimitsMin=<0 -15 -5>, angularTarget=<0 0 0>}
      localJointOffset=<2 0 0>
     }]
   }
   Save cached pose 'AnimDynamics'(Pose=Component To Local(ComponentPose=@n32)) {cacheName=AnimDynamics}
  ]}
 EventGraph {nodes [
   Event BlueprintUpdateAnimation()
   > n1 = Gravity Override State Function DEPRECATED(CharacterPartAnimInstance=Self-Reference(), GameTimeInSeconds=GetGameTimeInSeconds(), GravityOverrideParams=Get Gravity Override Params())
   > Set Gravity Override(Gravity Override=@n1)
   n2 = GetAlphaFromReferenceTransformDelta(SkeletalMeshComponent=GetOwningComponent(), BoneName=head, BaseName=spine_04, InputRange=(X=-20,Y=-55), TwistAxis="1,0,0")
   > n3 = Set HeadUp(HeadUp:float=@n2) -> (Output_Get:float)
   n4 = GetAlphaFromReferenceTransformDelta(
    SkeletalMeshComponent=GetOwningComponent()
    BoneName=head
    BaseName=spine_04
    DeltaComponentType=EulerX
    InputRange=(X=0,Y=-20)
    TwistAxis="1,0,0"
   )
   > Set HeadRight(HeadRight:float=MapRangeClamped(Value=@n4, InRangeA=0, InRangeB=1, OutRangeB=@n2)) -> (Output_Get:float)
   > Set HeadUpCrouching(HeadUpCrouching:double=SelectFloat(A=@n3, B=0, bPickA=Get bIsCrouching())) -> (Output_Get:double)
   > n5 = Branch(Condition=Get bIsSkyDiving())
   > Set HeadUpSkyDive(HeadUpSkyDive:double=Get HeadUp() -> (HeadUp:float)) -> (Output_Get:double)
   Set HeadUpSkyDive(execute=@n5.else, HeadUpSkyDive=0)
   n6 = Event OnCharacterCustomizationLoaded()
   > Set BodyMesh(BodyMesh:Skeletal Mesh Component Budgeted Object Reference=Get Array Item with EFort Custom Part Type(Array=@n6.CustomMeshParts)) -> (Output_Get:Skeletal Mesh Component Budgeted Object Reference)
   n7 = CheckForBackbling()
   > Set bIsWearingBackpack(bIsWearingBackpack:bool=OR Boolean(A=@n7, B=@n7.bIsBackblingLong)) -> (Output_Get:bool)
   TryGetPawnOwner()
   SelectFloat(A=@n4, B=0, bPickA=false)
  ]}
]
variables [
 bIsWearingBackpack {type=boolean}
 BodyMesh {type=SkeletalMeshComponentBudgeted}
 Gravity Override {type=Vector}
 Gravity Override Params {type=GravityOverrideParamsStruct}
 HeadRight {type=number}
 HeadUp {type=number}
 HeadUpCrouching {type=number}
 HeadUpSkyDive {type=number}
]
